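# Landscape material function for Athena grass, outputting material attributes, plain base color, alpha and UV scale.
# World-XY grass texture samples, optionally channel-remapped, are tinted by parameter colors through micro and macro masks from MF_GrassMicroColorMasks, linked by named reroutes.
# A virtual-texture switch selects between a rerouted color and GrassDistanceBlend RGB.
# Distance-blend alpha drives opacity and powered specular; a world-tiled normal map and roughness parameter complete the surface.
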MaterialFunction /FigmentCore/Game/Athena/Environments/Landscape/MaterialFunctions/Standard/MF_Athena_Grass_01
graph {
 properties {bExposeToLibrary=true}
 nodes [
  n1 = MaterialFunctionCall() {materialFunction=&LinearGradient}
  Add(
   A=@n1.UGradient
   B=Add(A=MaterialFunctionCall() {materialFunction=&DebugTimeSine}.Result) {constB=-0.5}
  )
  n2 = NamedRerouteUsage()
  n3 = VectorParameter() {group=Grass, parameterName=GrassMacroColorA, defaultValue=<rgba 0.4971 0.9792 0 1>}
  n4 = LinearInterpolate(
   A=LinearInterpolate(
    A=VectorParameter() {group=Grass, parameterName=SSC-1, defaultValue=<rgba 1 1 1 1>}.RGB
    B=VectorParameter() {group=Grass, parameterName=SSC-2, defaultValue=<rgba 1 1 1 1>}.RGB
    Alpha=NamedRerouteUsage()
   )
   B=VectorParameter() {group=Grass, parameterName=SSC-3, defaultValue=<rgba 0.9479 0.9048 0.077 1>}.RGB
   Alpha=NamedRerouteUsage()
  )
  n5 = NamedRerouteUsage()
  n6 = Multiply(A=@n5, B=Constant() {r=4})
  n7 = Multiply(A=@n2, B=LinearInterpolate(A=@n3.RGB, B=@n4, Alpha=Saturate(None=@n6)))
  Add(A=Multiply(B=ComponentMask(None=@n2) {r=true}) {constA=0.01}, B=@n7)
  Constant3Vector()
  Constant3Vector() {constant=<rgba 0.2683 0.9479 0.1975 1>}
  Constant3Vector() {constant=<rgba 0.138 0.7292 0 1>}
  Constant() {r=700}
  n8 = MaterialFunctionCall!(UseVertexInterp) {materialFunction=&/MF_Figment_Landscape_GrassDistanceBlend}
  n9 = NamedRerouteUsage()
  n10 = ScalarParameter() {defaultValue=0.15, group=Grass, parameterName=GrassSpecAmount}
  n11 = Power(Base=@n8.Grass A, Exp=ScalarParameter() {defaultValue=2, group=Grass, parameterName=GrassSpecPower})
  n12 = WorldPosition() {worldPositionShaderOffset=WPT_ExcludeAllShaderOffsets}
  n13 = TextureSample(UVs=Divide(A=ComponentMask(None=@n12.XYZ) {g=true, r=true}, B=ScalarParameter() {defaultValue=700, parameterName=GrassNearScale})) {
   samplerSource=SSM_Wrap_WorldGroupSettings
   samplerType=SAMPLERTYPE_Normal
   texture=/Game/Packages/DS_Fortnite_Terrain_NoLOD/Textures/T_Grasslands_N
  }
  FunctionOutput(None=MakeMaterialAttributes!(
    BaseColor=VirtualTextureFeatureSwitch!(No=@n8.Grass RGB, Yes=@n9)
    Metallic
    Specular=LinearInterpolate(B=@n10, Alpha=@n11)
    Roughness=ScalarParameter() {defaultValue=0.9, parameterName=Grass_Roughness}
    Anisotropy
    EmissiveColor
    Opacity=@n8.Grass A
    OpacityMask
    Normal=@n13.RGB
    Tangent
    WorldPositionOffset
    SubsurfaceColor
    ClearCoat
    ClearCoatRoughness
    AmbientOcclusion
    Refraction
    CustomizedUVs_0
    CustomizedUVs_1
    CustomizedUVs_2
    CustomizedUVs_3
    CustomizedUVs_4
    CustomizedUVs_5
    CustomizedUVs_6
    CustomizedUVs_7
    PixelDepthOffset
    ShadingModel
    Displacement
   ))
  FunctionOutput(None=VirtualTextureFeatureSwitch(No=@n8.Grass RGB, Yes=@n9)) {outputName=BaseColor_Plain}
  FunctionOutput(None=@n8.Grass A) {outputName=Alpha}
  n14 = NamedRerouteUsage()
  LinearInterpolate(A=@n4, B=VectorParameter() {group=Grass, parameterName=yellowish, defaultValue=<rgba 0.0346 0.4375 0.1261 1>}.RGB, Alpha=@n14)
  LinearInterpolate(
   A=@n3.RGB
   B=VectorParameter() {group=Grass, parameterName=GrassMacroColorB, defaultValue=<rgba 0.0784 0.349 0.0667 1>}.RGB
   Alpha=LinearInterpolate(
    A=ScalarParameter() {group=Grass, parameterName=MacroGrassBlend_Min}
    B=ScalarParameter() {defaultValue=1, group=Grass, parameterName=MacroGrassBlend_Max}
    Alpha=@n5
   )
  )
  n15 = ComponentMask(None=WorldPosition() {worldPositionShaderOffset=WPT_ExcludeAllShaderOffsets}.XYZ) {g=true, r=true}
  n16 = ScalarParameter() {defaultValue=700, parameterName=GrassNearScale}
  n17 = TextureSampleParameter2D(UVs=FunctionOutput(None=Divide(A=@n15, B=@n16)) {outputName=GrassScale}) {
   group=Grass
   parameterName=Grass
   samplerSource=SSM_Wrap_WorldGroupSettings
   texture=/Game/Athena/Environments/Landscape/Textures/T_GrasslandsChannelNormalized
  }
  n18 = OneMinus(None=Power(Base=OneMinus(None=@n17.R)) {constExponent=2.2})
  n19 = OneMinus(None=Power(Base=OneMinus(None=@n17.G)) {constExponent=2.2})
  n20 = OneMinus(None=Power(Base=OneMinus(None=@n17.B)))
  n21 = StaticSwitchParameter(
   True=MaterialFunctionCall(
    X=LinearInterpolate(Alpha=@n18) {constA=0.04, constB=0.11}
    Y=LinearInterpolate(Alpha=@n19) {constA=0.12, constB=0.2}
    Z=LinearInterpolate(Alpha=@n20) {constA=0.015, constB=0.05}
   ) {materialFunction=&MakeFloat3}.Result
   False=@n17.RGB
  ) {defaultValue=true, parameterName=Use Remap Grass}
  LinearInterpolate(A=@n9, B=@n21, Alpha=ScalarParameter() {group=Grass, parameterName=DEbug_Old})
  NamedRerouteDeclaration(None=Saturate(None=LinearInterpolate(A=@n7, B=@n2, Alpha=@n14))) {name=Final_Color, nodeColor=<rgba 0 1 0.1094 1>}
  NamedRerouteDeclaration(None=@n21) {name=Base Color, nodeColor=<rgba 0.3594 0 1 1>}
  n22 = MaterialFunctionCall!(SamplingPosition=WorldPosition() {worldPositionShaderOffset=WPT_ExcludeAllShaderOffsets}.XYZ) {materialFunction=/Game/Athena/Asteria/Environments/Landscape/Materials/MaterialFunctions/MF_GrassMicroColorMasks}
  NamedRerouteDeclaration(None=@n22.GrassColorMask3) {name=GrassColorMask3, nodeColor=<rgba 1 0 0.8203 1>}
  NamedRerouteDeclaration(None=@n22.GrassColorMask4) {name=GrassColorMask4, nodeColor=<rgba 1 0 0.8203 1>}
  NamedRerouteDeclaration(None=@n22.GrassMacroColorMask) {name=GrassMacroColorMask, nodeColor=<rgba 1 0 0.8203 1>}
  NamedRerouteDeclaration(None=@n22.GrassColorMask2) {name=GrassColorMask2, nodeColor=<rgba 1 0 0.8203 1>}
  NamedRerouteDeclaration(None=@n22.GrassColorMask1) {name=GrassColorMask1, nodeColor=<rgba 1 0 0.8203 1>}
  Round(None=@n1.UGradient)
  TextureSample() {texture=/Game/Athena/Environments/Landscape/Textures/T_Athena_Grass_Farm_ColorMatched_D_2}
 ]
}
